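# Masked unlit Niagara particle material for the Digital Graffiti aura's spline arrows, with iridescent holofoil color.
# Dynamic parameters control arrow texture pan and scale plus SplineThicken width.
# Panning, particle-randomized Perlin noise warps time-scrolled stripes indexing a curve atlas, tinted by particle color, hue-shifted and masked by texture channels, then scaled by time of day for emissive.
# Opacity combines particle alpha, texture mask and a noise-broken vertical gradient.
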Material /SparksCosmetics/Items/Auras/Assets/NS_Aura_DigitalGraffiti/M_Aura_DigitalGraffiti_Arow_Splines
graph {
 properties {
  bAutomaticallySetUsageInEditor=false
  bUsedWithNiagaraMeshParticles=true
  bUsedWithNiagaraSprites=true
  blendMode=BLEND_Masked
  shadingModel=MSM_Unlit
  usedShadingModels=Unlit
 }
 nodes [
  Constant() {desc=Scruntch texture so it is the right proportion, r=0.5}
  n1 = DynamicParameter() {defaultValue=<rgba 1 1 0.5 7>, paramNames [TexturePan, TextureThickness, TextureScale, SplineThickness]}
  MaterialFunctionCall(WidthBase=@n1.SplineThickness, WidthTip=@n1.SplineThickness) {materialFunction=&SplineThicken}
  n2 = ScalarParameter() {defaultValue=0.9, group=TOD Color, parameterName=Morn}
  n3 = ScalarParameter() {defaultValue=1, group=TOD Color, parameterName=Day}
  n4 = ScalarParameter() {defaultValue=0.9, group=TOD Color, parameterName=Evening}
  n5 = ScalarParameter() {defaultValue=0.6, group=TOD Color, parameterName=Night}
  n6 = AppendVector(A=@n1.TextureThickness, B=@n1.TextureScale)
  n7 = Constant()
  n8 = TextureSampleParameter2D(UVs=Add(
    A=MaterialFunctionCall(Texture Scale=@n6) {materialFunction=&ScaleUVsByCenter}.UVs
    B=AppendVector(A=@n7, B=@n1.TexturePan)
   ) {constB=0}) {
   parameterName=Texture
   samplerSource=SSM_Clamp_WorldGroupSettings
   texture=&/T_Aura_DigitalGraffiti_Arrow
  }
  n9 = Multiply(
   A=TextureCoordinate()
   B=AppendVector(
    A=ScalarParameter() {defaultValue=0.5, group=Distortion Texture Parameters, parameterName=Tiling X, sliderMax=2}
    B=ScalarParameter() {defaultValue=1, group=Distortion Texture Parameters, parameterName=Tiling Y, sliderMax=2}
   )
  )
  n10 = ParticleRandom()
  n11 = ScalarParameter() {defaultValue=0.05, group=Distortion Texture Parameters, parameterName=Speed X, sliderMax=5, sliderMin=-5}
  n12 = ScalarParameter() {defaultValue=-0.1, group=Distortion Texture Parameters, parameterName=Speed Y, sliderMax=5, sliderMin=-5}
  n13 = TextureSample(UVs=Panner(Coordinate=Add(A=@n9, B=@n10), Speed=AppendVector(A=@n11, B=@n12)) {speedX=1}) {samplerType=SAMPLERTYPE_LinearGrayscale, texture=&T_SimplePerlin_Greyscale}
  n14 = Add(
   A=TextureCoordinate()
   B=Multiply(
    A=Subtract(A=@n13.R) {constB=0.5}
    B=ScalarParameter() {defaultValue=0.2, group=Distortion Texture Parameters, parameterName=Distortion Strength, sliderMax=1}
   ) {constB=0.1}
  )
  n15 = Sine(None=Multiply(
    A=ComponentMask(None=@n14) {g=true}
    B=ScalarParameter() {defaultValue=2, group=Foil Color Stripes, parameterName=Stripe Density, sliderMax=20}
   ) {constB=10}) {period=1.5}
  n16 = Multiply(
   A=Saturate(None=ConstantBiasScale(None=@n15))
   B=ScalarParameter() {defaultValue=0.5, group=Foil Color Stripes, parameterName=Color Width, sliderMax=5}
  ) {constB=3}
  n17 = Divide(A=Time())
  n18 = CurveAtlasRowParameter(CurveTime=Frac(None=Add(A=@n16, B=@n17))) {
   atlas=/Game/Effects/Niagara/MasterMaterials/ShockWave/Texture/CA_Crew_Holofoil_Iridescent
   curve=/Game/Effects/Niagara/MasterMaterials/ShockWave/Texture/Curve_Crew_Holofoil_Iridescent
   group=Foil Color Stripes
   parameterName=CurveAtlas
  }
  n19 = ParticleColor()
  n20 = MaterialFunctionCall(
   Hue Shift Percentage=Multiply(A=StaticComponentMaskParameter(None=@n8.RGBA) {defaultG=true, group=Texture Channel Masks, parameterName=Hue Shift Channel RBGA}) {constB=0.1}
   Texture=Multiply(A=@n18, B=@n19.RGB)
  ) {materialFunction=&HueShift}
  n21 = StaticComponentMaskParameter(None=@n8.RGBA) {defaultR=true, group=Texture Channel Masks, parameterName=Foil Shape Channel RGBA}
  MaterialFunctionCall(Emissive=Multiply(
    A=MaterialFunctionCall(Color_Morning=@n2, Color_Day=@n3, Color_Evening=@n4, Color_Night=@n5) {materialFunction=&Color_Time_of_Day}.Result
    B=Multiply(A=@n20.Result, B=@n21)
   )) {materialFunction=&MF_FX_EmissiveAdjustments}
  n22 = ComponentMask(None=OneMinus(None=TextureCoordinate())) {g=true}
  n23 = Multiply(A=@n22, B=@n13.R)
  n24 = Divide(A=Subtract(A=Add(A=@n22, B=@n23), B=Constant() {r=0.75}), B=Constant() {r=1.5})
  Multiply(
   A=@n19.A
   B=Multiply(
    A=StaticComponentMaskParameter(None=@n8.RGBA) {defaultA=true, group=Texture Channel Masks, parameterName=Mask Opacity RGBA}
    B=Saturate(None=@n24)
   )
  )
 ]
}
