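# Translucent unlit Niagara sprite and ribbon material for the icons of the Digital Graffiti aura cosmetic.
# Perlin noise, panned and offset per particle, distorts UVs into wavy sine stripes that, with time, index an animated iridescent holofoil curve atlas.
# The gradient is tinted by particle color, hue-shifted by the icon texture's green channel and masked by red.
# Time-of-day scalars scale brightness, while particle and texture alpha multiply together.
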Material /SparksCosmetics/Items/Auras/Assets/NS_Aura_DigitalGraffiti/M_Aura_DigitalGraffiti_Icons
graph {
 properties {
  bAutomaticallySetUsageInEditor=false
  bUsedWithNiagaraRibbons=true
  bUsedWithNiagaraSprites=true
  blendMode=BLEND_Translucent
  shadingModel=MSM_Unlit
  usedShadingModels=Unlit
 }
 nodes [
  n1 = ScalarParameter() {defaultValue=0.9, parameterName=Morn}
  n2 = ScalarParameter() {defaultValue=1, parameterName=Day}
  n3 = ScalarParameter() {defaultValue=0.9, parameterName=Evening}
  n4 = ScalarParameter() {defaultValue=0.6, parameterName=Night}
  n5 = TextureSample() {texture=&/T_Aura_DigitalGraffiti_Icons_3x3}
  n6 = Multiply(
   A=TextureCoordinate()
   B=AppendVector(
    A=ScalarParameter() {defaultValue=0.5, parameterName=Tiling X, sliderMax=2}
    B=ScalarParameter() {defaultValue=1, parameterName=Tiling Y, sliderMax=2}
   )
  )
  n7 = ParticleRandom()
  n8 = ScalarParameter() {defaultValue=0.1, parameterName=Speed X, sliderMax=5, sliderMin=-5}
  n9 = ScalarParameter() {defaultValue=-0.1, parameterName=Speed Y, sliderMax=5, sliderMin=-5}
  n10 = TextureSample(UVs=Panner(Coordinate=Add(A=@n6, B=@n7), Speed=AppendVector(A=@n8, B=@n9)) {speedX=1}) {samplerType=SAMPLERTYPE_LinearGrayscale, texture=&T_SimplePerlin_Greyscale}
  n11 = Add(
   A=TextureCoordinate()
   B=Multiply(A=Subtract(A=@n10.R) {constB=0.5}, B=ScalarParameter() {defaultValue=0.2, parameterName=Distortion Strength, sliderMax=1}) {constB=0.1}
  )
  n12 = Sine(None=Multiply(A=ComponentMask(None=@n11) {g=true}, B=ScalarParameter() {defaultValue=2, parameterName=Stripe Density, sliderMax=20}) {constB=10}) {period=1.5}
  n13 = Multiply(A=Saturate(None=ConstantBiasScale(None=@n12)), B=ScalarParameter() {defaultValue=0.5, parameterName=Color Width, sliderMax=5}) {constB=3}
  n14 = Add(A=Time(), B=@n7)
  n15 = CurveAtlasRowParameter(CurveTime=Frac(None=Add(A=@n13, B=@n14))) {
   atlas=/Game/Effects/Niagara/MasterMaterials/ShockWave/Texture/CA_Crew_Holofoil_Iridescent
   curve=/Game/Effects/Niagara/MasterMaterials/ShockWave/Texture/Curve_Crew_Holofoil_Iridescent
   parameterName=CurveAtlas
  }
  n16 = ParticleColor()
  n17 = MaterialFunctionCall(
   Hue Shift Percentage=Multiply(A=@n5.G, B=ScalarParameter() {defaultValue=0.1, parameterName=Hue Shift}) {constB=0.1}
   Texture=Multiply(A=@n15, B=@n16.RGB)
  ) {materialFunction=&HueShift}
  MaterialFunctionCall(Emissive=Multiply(
    A=MaterialFunctionCall(Color_Morning=@n1, Color_Day=@n2, Color_Evening=@n3, Color_Night=@n4) {materialFunction=&Color_Time_of_Day}.Result
    B=Multiply(A=@n17.Result, B=@n5.R)
   )) {materialFunction=&MF_FX_EmissiveAdjustments}
  Multiply(A=@n16.A, B=@n5.A)
  n18 = TextureCoordinate()
  n19 = Power(Base=OneMinus(None=ComponentMask(None=@n18) {g=true})) {constExponent=5}
  Multiply(A=Add(A=Multiply(A=@n19) {constB=3}), B=@n15)
 ]
}
