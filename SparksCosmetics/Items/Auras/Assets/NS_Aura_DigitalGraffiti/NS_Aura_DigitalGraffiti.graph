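# Niagara system for the Digital Graffiti aura cosmetic, layering neon arrows, graffiti icons, glows, rays, speed lines, stars and dots on a skeletal mesh.
# Emitters place particles at bones or sockets using partition-indexed arrays; icons flip SubUV frames and ease outward through interpolated positions.
# IsLargeBody adjusts sizes and offsets, IsFrontEnd tunes alpha and offsets, and time-of-day scalars balance brightness.
# Glow also carries the rays; standalone Rays and Glow001 are disabled.
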NiagaraSystem /SparksCosmetics/Items/Auras/Assets/NS_Aura_DigitalGraffiti/NS_Aura_DigitalGraffiti
dependencies {
 usedDataInterfaces [
  &NiagaraDataInterfaceArrayFloat
  &NiagaraDataInterfaceArrayInt32
  &NiagaraDataInterfaceCurve
  &NiagaraDataInterfaceColorCurve
  &NiagaraDataInterfaceArrayFloat3
  &NiagaraDataInterfaceSpriteRendererInfo
  &NiagaraDataInterfaceVectorCurve
 ]
 usedDynamicInputs [
  &SimulationPosition
  &MakeCustomLinearColorFromBool
  &RandomBool
  &VectorFromCurve
 ]
 usedModules [
  &self:SystemSpawnScript.SystemScriptSource.SystemScriptGraph.NiagaraNodeAssignment_1.SetVariables_#1
  &SystemState
 ]
}
emitters {
 Arrows001 {
  emitterUpdateScript [
   &EmitterState(Life Cycle Mode=::NewEnumerator1, Loop Duration=1, UseLoopDelay=false)
   &SpawnRate(Spawn Probability=0.8, SpawnRate=2, Use Spawn Probability=true)
  ]
  particleSpawnScript [
   &PartitionParticles(Number Of Partitions=5)
   &InitializeParticle(
    Color=&ColorFromCurve(
     ColorCurve=DataInterface() {propertyValues {
       AlphaCurve {keys [<at 0 1 RCIM_Cubic>]}
       BlueCurve {keys [<at 0.0081 1>,<at 0.196 1 RCIM_Cubic>,<at 1 1 RCIM_Cubic>]}
       GreenCurve {keys [<at 0.0081 0.1133>, {arriveTangent=0.8789, interpMode=RCIM_Cubic, leaveTangent=0.8789, time=0.196, value=0.294}, <at 1 0.9851 RCIM_Cubic>]}
       RedCurve {keys [<at 0.0081 0>, {arriveTangent=0.0599, interpMode=RCIM_Cubic, leaveTangent=0.0599, time=0.196, value=0.0802}, <at 1 0.0594 RCIM_Cubic>]}
      }}
     CurveIndex=&RandomRangeFloat(Minimum=0, Maximum=1)
    )
    Lifetime=&RandomRangeFloat(Minimum=0.9, Maximum=1.2)
    Mesh Scale Mode=::NewEnumerator3
    Mesh Uniform Scale=&RandomRangeFloat(Minimum=0.6, Maximum=0.65)
    Position Offset=&RandomRangeVector(Minimum=<0 2 -1>, Maximum=<0 2 -1>)
    UsePositionOffset=false
    Write Material Random=true
   )
   &SkeletalMeshLocation(
    Bone / Socket Index=&SelectIntFromArray(
     Direct Array Index=$Particles.PartitionParticles.Partition
     Direct Array Mode=::NewEnumerator1
     Int Selection Array=DataInterface() {propertyValues {IntData [7]}}
    )
    Bone Sampling Mode=::NewEnumerator4
    MaskSpawnGroup=false
    Sampled Position Offset=&Add_Vector(
     A=&RandomRangeVector(Minimum=<-1 0 -1>, Maximum=<0.5 1 1>)
     B=&Multiply_Vector(
      A=&SelectVectorFromArray(
       Direct Array Index=$Particles.PartitionParticles.Partition
       Vector Selection Array=DataInterface() {propertyValues {FloatData [<6 2 4.2696>, <2 1 1>, <2 0 -3.5634>, <2 0 -1>, <2 1 2.3751>]}}
      )
      B=&MakeCustomVectorFromBool(Bool=$User.IsLargeBody, True Vector=<1.1 1 1.1>, False Vector=<1 1 1>)
     )
    )
    Skeletal Mesh=$System.Skeletal Mesh
   )
   &InitialMeshOrientation(
    Mesh Orientation Mode=::NewEnumerator2
    Rotation=&SelectVectorFromArray(
     Direct Array Index=$Particles.PartitionParticles.Partition
     Vector Selection Array=DataInterface() {propertyValues {FloatData [<0 0 0>, <0.14 0 0>, <-0.14 0 0>, <0 0 0>, <0.1 0 0>]}}
    )
    Use Rotation Vector=true
   )
   &CameraOffset(Camera Offset Amount=-115)
   /SetVariables(Particles.MeshIndex=&SelectIntFromArray(
     Array Sampling Mode=::NewEnumerator1
     Direct Array Index=?
     Int Selection Array=DataInterface() {propertyValues {IntData [0,1,1]}}
     Random Seed=0
    )) {isSetParameters=true}
  ]
  particleUpdateScript [
   &ParticleState()
   &UpdateMeshOrientation(Rotation Rate=&FloatFromCurve(
     FloatCurve=DataInterface() {propertyValues {Curve {defaultValue=1, keys [<at 0 0 RCIM_Cubic>,<at 0.8 1 RCIM_Cubic>]}}}
     Scale Curve=&RandomRangeFloat(Minimum=0, Maximum=0.5)
    ))
   &ScaleMeshSize(Scale Factor=&MakeVector(
     X=&FloatFromCurve(FloatCurve=DataInterface() {propertyValues {Curve {
         defaultValue=1
         keys [
          {arriveTangent=0.9718, interpMode=RCIM_Cubic, leaveTangent=0.9718, tangentMode=RCTM_User, time=-0.0015, value=0.7039}
          {
           arriveTangentWeight=0.2779
           interpMode=RCIM_Cubic
           leaveTangent=1.2484
           leaveTangentWeight=0.129
           tangentMode=RCTM_Break
           time=0.85
           value=0.96
          }
          {arriveTangent=0.7876, interpMode=RCIM_Cubic, leaveTangent=0.7876, tangentMode=RCTM_User, time=1.0059, value=1.1}
         ]
        }}})
     Y=&FloatFromCurve(
      FloatCurve=DataInterface() {propertyValues {Curve {keys [{arriveTangent=-3.3355, interpMode=RCIM_Cubic, leaveTangent=-3.3355, tangentMode=RCTM_User, time=0.55, value=1}, <at 0.9 0.35 RCIM_Cubic>]}}}
      Scale Curve=&RandomRangeFloat(Minimum=1, Maximum=1.35)
     )
     Z=0.1
    )) {module=ScaleMeshSize001}
   &ScaleMeshSize(Scale Factor=&SelectVectorFromArray(
     Array Sampling Mode=::NewEnumerator0
     Direct Array Index=$Particles.PartitionParticles.Partition
     Direct Array Mode=::NewEnumerator1
     Vector Selection Array=DataInterface() {propertyValues {FloatData [<1 1 1>, <1 -1 1>]}}
     Random Seed=0
    ))
   &DynamicMaterialParameters(
    Index 0 Param 1=&FloatFromCurve(FloatCurve=DataInterface() {propertyValues {Curve {keys [{arriveTangent=2.4275, interpMode=RCIM_Cubic, leaveTangent=2.4275, tangentMode=RCTM_User, time=0, value=-0.3824}, <at 0.85 0.53 RCIM_Cubic>]}}})
    Index 0 Param 2=&FloatFromCurve(FloatCurve=DataInterface() {propertyValues {Curve {
        defaultValue=1
        keys [
         {arriveTangent=1.0445, interpMode=RCIM_Cubic, leaveTangent=1.0445, tangentMode=RCTM_User, time=0, value=0.2549}
         {arriveTangent=1.0359, interpMode=RCIM_Cubic, leaveTangent=1.0359, time=0.5318, value=0.8327}
         {arriveTangent=0.2311, interpMode=RCIM_Cubic, leaveTangent=0.2311, tangentMode=RCTM_User, time=0.7, value=0.98}
         {arriveTangent=-0.0323, interpMode=RCIM_Cubic, leaveTangent=-0.0323, tangentMode=RCTM_User, time=0.8, value=1}
         {
          arriveTangent=-1.0291
          arriveTangentWeight=0.062
          interpMode=RCIM_Cubic
          leaveTangent=-12.0481
          leaveTangentWeight=0.1303
          tangentMode=RCTM_Break
          time=0.86
          value=0.98
         }
         {arriveTangentWeight=0.0414, interpMode=RCIM_Cubic, tangentMode=RCTM_User, time=0.98, value=0.01}
        ]
       }}})
    Index 0 Param 3=&FloatFromCurve(FloatCurve=DataInterface() {propertyValues {Curve {keys [
         {arriveTangent=-1.2589, interpMode=RCIM_Cubic, leaveTangent=-1.2589, tangentMode=RCTM_User, time=0.002, value=0.8738}
         {
          arriveTangent=-0.2838
          arriveTangentWeight=0.2012
          interpMode=RCIM_Cubic
          leaveTangent=-0.2838
          leaveTangentWeight=0.1035
          tangentMode=RCTM_User
          time=0.5921
          value=0.5
         }
         {
          arriveTangent=0.0886
          arriveTangentWeight=0.1057
          interpMode=RCIM_Cubic
          leaveTangent=0.0886
          leaveTangentWeight=0.0345
          tangentMode=RCTM_User
          time=0.85
          value=0.47
         }
         {arriveTangent=2.2765, interpMode=RCIM_Cubic, leaveTangent=2.2765, tangentMode=RCTM_User, time=1, value=0.637}
        ]}}})
    Index 0 Param 4=&RandomRangeFloat(Minimum=18, Maximum=32)
    Module.Index0Float0Write=true
    Module.Index0Float1Write=true
    Module.Index0Float2Write=true
    Module.Index0Float3Write=true
   ) {module=DynamicMaterialParameters001}
   &ScaleColor(
    Scale RGB=&VectorFromFloat(Value=&FloatFromCurve(
      FloatCurve=DataInterface() {propertyValues {Curve {keys [
          <at 0.2 1 RCIM_Cubic>
          {arriveTangent=0.6124, interpMode=RCIM_Cubic, leaveTangent=0.6124, tangentMode=RCTM_User, time=0.75, value=1.3679}
          <at 0.95 1 RCIM_Cubic>
         ]}}}
      Scale Curve=&FNTimeOfDayScalar(Night Scale=1.15, Morning Scale=0.85, Day Scale=0.6, Evening Scale=0.7)
     ))
    ScaleA=true
    ScaleRGB=true
   )
   &Color(Color=&ColorFromCurve(ColorCurve=DataInterface() {propertyValues {
       AlphaCurve {keys [<at 0 1 RCIM_Cubic>,<at 1 1 RCIM_Cubic>]}
       BlueCurve {keys [<at 0 0.354 RCIM_Cubic>,<at 1 1 RCIM_Cubic>]}
       GreenCurve {keys [<at 0 0 RCIM_Cubic>,<at 1 0.364 RCIM_Cubic>]}
       RedCurve {keys [<at 0 2 RCIM_Cubic>,<at 1 0 RCIM_Cubic>]}
      }})) {enabled=false}
   &Drag(Drag=2, Use Linear Drag=true, Use Rotational Drag=true) {enabled=false}
   &SolveForcesAndVelocity(Clamp Velocity=false, Limit Acceleration=false)
   /SetVariables(Particles.Position=&AddVectorToPosition(
     Position=$Particles.Position
     Vector=&MakeVector(
      X=0
      Y=0
      Z=&FloatFromCurve(
       FloatCurve=DataInterface() {propertyValues {Curve {
          defaultValue=1
          keys [
           {interpMode=RCIM_Cubic, tangentMode=RCTM_User, time=0, value=0}
           {arriveTangent=2.315, interpMode=RCIM_Cubic, leaveTangent=2.315, tangentMode=RCTM_User, time=1, value=1}
          ]
         }}}
       Scale Curve=-0.45
      )
     )
    )) {isSetParameters=true}
   /SetVariables(
    Particles.Linear Color=&ColorFromCurve(ColorCurve=DataInterface() {propertyValues {
       AlphaCurve {keys [<at 0.0998 0>,<at 0.8415 1>,<at 0.9632 0 RCIM_Cubic>]}
       BlueCurve {keys [
         <at 0.4403 1>
         {arriveTangent=-0.536, interpMode=RCIM_Cubic, leaveTangent=-0.536, time=0.6171, value=1}
         <at 0.7883 0.8135>
         <at 0.8472 0 RCIM_Cubic>
         <at 0.929 1>
        ]}
       GreenCurve {keys [
         <at 0.4403 0>
         {arriveTangent=2.8487, interpMode=RCIM_Cubic, leaveTangent=2.8487, time=0.6171, value=0.1888}
         <at 0.7883 0.9914>
         {interpMode=RCIM_Cubic, leaveTangent=-0.2534, time=0.8472, value=1}
         <at 0.929 0.1102>
        ]}
       RedCurve {keys [
         <at 0.4403 0.2611>
         {arriveTangent=-0.7501, interpMode=RCIM_Cubic, leaveTangent=-0.7501, time=0.6171, value=0}
         <at 0.7883 0>
         <at 0.8472 0.2157 RCIM_Cubic>
         <at 0.929 0>
        ]}
      }})
    Particles.cameraoffset2:NiagaraFloat=&FloatFromCurve(FloatCurve=DataInterface() {propertyValues {Curve {keys [<at 0 115>,<at 0.8 110>]}}}, Scale Curve=-1)
    Particles.positionOffset:NiagaraPosition=&AddVectorToPosition(
     Position=$Particles.Position
     Vector=&Multiply_VectorByFloat(
      Vector=<0.25 0 0.5>
      Float=&FloatFromCurve(FloatCurve=DataInterface() {propertyValues {Curve {
          defaultValue=1
          keys [
           {arriveTangent=-0.1262, interpMode=RCIM_Cubic, leaveTangent=-0.1262, tangentMode=RCTM_User, time=0, value=0}
           {arriveTangent=1.5711, interpMode=RCIM_Cubic, leaveTangent=1.5711, tangentMode=RCTM_User, time=1, value=1}
          ]
         }}})
     )
    )
   ) {isSetParameters=true}
  ]
  rendererClasses [&NiagaraMeshRendererProperties]
  renderers [
   {rendererClass=&NiagaraMeshRendererProperties, rendererIndex=0}
   {rendererClass=&NiagaraMeshRendererProperties, rendererIndex=1}
  ]
 }
 Glow {
  emitterUpdateScript [
   &EmitterState()
   &SpawnBurst_Instantaneous(Spawn Count=7, Use Spawn Probability=false, UseLoopCountLimit=true)
   &SpawnBurst_Instantaneous(Spawn Group=1, Use Spawn Probability=false, UseLoopCountLimit=true) {module=SpawnBurst_Instantaneous001}
  ]
  particleSpawnScript [
   /SetVariables(Particles.VisibilityTag=$Emitter.SpawnGroup) {isSetParameters=true}
   &PartitionParticles(Number Of Partitions=7)
   &InitializeParticle(
    Color=&Multiply_LinearColorByFloat(
     Linear Color=<rgba 2.75 0.0279 2.2441 1>
     Float=&FNTimeOfDayScalar(Night Scale=0.7, Morning Scale=0.7, Day Scale=0.8, Evening Scale=0.9)
    )
    Mesh Scale Mode=::NewEnumerator3
    Mesh Uniform Scale=&MakeCustomFloatFromBool(Bool=$User.IsLargeBody, True Float=1.6, False Float=1.5)
    Sprite Rotation Mode=::NewEnumerator0
    Sprite Size Mode=::NewEnumerator3
    Uniform Sprite Size=&Multiply_Float(
     A=&SelectFloatFromArray(
      Direct Array Index=&SpawnGroup()
      Float Selection Array=DataInterface() {propertyValues {FloatData [40,40,20,20,30,30]}}
     )
     B=&MakeCustomFloatFromBool(Bool=$User.IsLargeBody, True Float=1.5, False Float=1)
    )
    UsePositionOffset=false
    Write Material Random=true
   ) {module=InitializeParticle001}
   /SetVariables(Particles.glowScale:Vector2f=&Vector2DFromFloat(Value=&Multiply_Float(
      A=&SelectFloatFromArray(
       Direct Array Index=$Particles.PartitionParticles.Partition
       Float Selection Array=DataInterface() {propertyValues {FloatData [90,70,70,70,70,40,40]}}
      )
      B=&MakeCustomFloatFromBool(Bool=$User.IsLargeBody, True Float=1.2, False Float=1)
     ))) {isSetParameters=true, module=SetVariables_#2}
   /SetVariables(Particles.RaysCameraOffset=-60, Particles.glowCameraOffset=-20) {isSetParameters=true, module=SetVariables_#3}
   /SetVariables() {isSetParameters=true, module=SetVariables_#4}
   /SetVariables(Particles.RaysDynamicParameter:Vector4f=&MakeVector4(
     X=&RandomRangeFloat(Minimum=0, Maximum=10)
     Y=&RandomRangeFloat(Minimum=1, Maximum=2)
     Z=&RandomRangeFloat(Minimum=0.8, Maximum=1.6)
     W=0
    )) {isSetParameters=true}
   &InitialMeshOrientation(Mesh Orientation Mode=::NewEnumerator2, Rotation=<0.5 0 0>, Use Rotation Vector=true)
  ]
  particleUpdateScript [
   &ParticleState(
    Kill Particles When Lifetime Has Elapsed=false
    Let Infinitely Lived Particles Die When Emitter Deactivates=true
    Loop Particles Lifetime=true
   )
   &CameraOffset(Camera Offset Amount=5)
   &SpriteRotationRate(Rotation Rate=&RandomRangeFloat(Minimum=10, Maximum=20))
   &SkeletalMeshLocation(
    Bone / Socket Index=&MaskIntBySpawnGroup(
     Masked Int=&SelectIntFromArray(Direct Array Index=$Particles.PartitionParticles.Partition, Int Selection Array=DataInterface() {propertyValues {IntData [0,1,2,3,4,5,6]}})
     Passthrough Int=7
     Spawn Group Mask=0
    )
    Bone Sampling Mode=::NewEnumerator4
    MaskSpawnGroup=false
    Orientation Sampling=::NewEnumerator1
    Sampled Position Offset Coordinate Space:ENiagaraSkelMeshTransforms=::NewEnumerator1
    Skeletal Mesh=$System.Skeletal Mesh
   ) {module=SkeletalMeshLocation003}
   &DynamicMaterialParameters(
    Index 0 Param 1=&SelectFloatFromArray(Direct Array Index=$Particles.PartitionParticles.Partition, Float Selection Array=DataInterface() {propertyValues {FloatData [5,5,4,4,3,3]}})
    Index 0 Param 3=1
    Module.Index0Float0Write=true
    Module.Index0Float1Write=true
    Module.Index0Float2Write=false
    Module.Index0Float3Write=true
   )
   /SetVariables(
    Particles.RaysColor:LinearColor=&MakeLinearColorFromVectorAndFloat(
     "Vector (RGB)"=&Multiply_VectorByFloat(
      Vector=&MakeVectorFromLinearColorRGB(Color=&ColorFromCurve(
        ColorCurve=DataInterface() {propertyValues {
          AlphaCurve {keys []}
          BlueCurve {keys [<at 0 1 RCIM_Cubic>,<at 0.2994 1.2>,<at 0.5042 1>,<at 0.7224 1.2>,<at 1 1 RCIM_Cubic>]}
          GreenCurve {keys [<at 0 0.4342 RCIM_Cubic>,<at 0.2994 0.4103>,<at 0.5042 0.8482>,<at 0.7224 0.4103>,<at 1 0.4342 RCIM_Cubic>]}
          RedCurve {keys [<at 0 0.9387 RCIM_Cubic>,<at 0.2994 0>,<at 0.5042 0.1742>,<at 0.7224 0>,<at 1 0.9387 RCIM_Cubic>]}
         }}
        CurveIndex=&Frac_Float(Float=&Multiply_Float(A=$Engine.Time, B=0.5))
       ))
      Float=&Multiply_Float(
       A=&FNTimeOfDayScalar(Night Scale=1.1, Morning Scale=1.1, Day Scale=1.1, Evening Scale=1.1)
       B=1.25
      )
     )
     "Float (Alpha)"=&MakeCustomFloatFromBool(Bool=$User.IsFrontEnd, True Float=0.21, False Float=0.6)
    )
    Particles.glowColor:LinearColor=&MakeLinearColorFromVectorAndFloat(
     "Vector (RGB)"=&MakeVectorFromLinearColorRGB(Color=<rgba 0.4031 2.0469 3 1>)
     "Float (Alpha)"=&FloatFromCurve(
      FloatCurve=DataInterface() {propertyValues {Curve {keys [<at 0 1>,<at 1 1>]}}}
      Scale Curve=&MakeCustomFloatFromBool(Bool=$User.IsFrontEnd, True Float=0.12, False Float=0.15)
     )
    )
   ) {isSetParameters=true, module=SetVariables_#5}
  ]
  rendererClasses [&NiagaraSpriteRendererProperties, &NiagaraMeshRendererProperties]
  renderers [
   {rendererClass=&NiagaraSpriteRendererProperties}
   {rendererClass=&NiagaraSpriteRendererProperties, rendererIndex=1}
   {rendererClass=&NiagaraMeshRendererProperties, rendererIndex=2}
  ]
 }
 Glow001 {
  bEnabled=false
  emitterUpdateScript [
   &EmitterState()
   &SpawnBurst_Instantaneous(Spawn Count=7, Use Spawn Probability=false, UseLoopCountLimit=true)
  ]
  particleSpawnScript [
   &PartitionParticles(Number Of Partitions=7)
   &InitializeParticle(
    Color=&Multiply_LinearColorByFloat(
     Linear Color=<rgba 2.75 0.0279 2.2441 1>
     Float=&FNTimeOfDayScalar(Night Scale=0.7, Morning Scale=0.7, Day Scale=0.8, Evening Scale=0.9)
    )
    Lifetime=3
    Sprite Rotation Mode=::NewEnumerator0
    Sprite Size Mode=::NewEnumerator3
    Uniform Sprite Size=&Multiply_Float(
     A=&SelectFloatFromArray(
      Direct Array Index=&SpawnGroup()
      Float Selection Array=DataInterface() {propertyValues {FloatData [40,40,20,20,30,30]}}
     )
     B=&MakeCustomFloatFromBool(Bool=$User.IsLargeBody, True Float=1.5, False Float=1)
    )
    UsePositionOffset=false
    Write Material Random=true
   ) {module=InitializeParticle001}
   /SetVariables(Particles.glowScale:Vector2f=&Vector2DFromFloat(Value=&Multiply_Float(
      A=&SelectFloatFromArray(
       Direct Array Index=$Particles.PartitionParticles.Partition
       Float Selection Array=DataInterface() {propertyValues {FloatData [90,70,70,70,70,40,40]}}
      )
      B=&MakeCustomFloatFromBool(Bool=$User.IsLargeBody, True Float=1.2, False Float=1)
     ))) {isSetParameters=true, module=SetVariables_#2}
   /SetVariables(Particles.glowCameraOffset=-20) {isSetParameters=true, module=SetVariables_#3}
   /SetVariables() {isSetParameters=true, module=SetVariables_#4}
  ]
  particleUpdateScript [
   &ParticleState(
    Kill Particles When Lifetime Has Elapsed=false
    Let Infinitely Lived Particles Die When Emitter Deactivates=true
    Loop Particles Lifetime=true
   )
   &CameraOffset(Camera Offset Amount=5)
   &SpriteRotationRate(Rotation Rate=&RandomRangeFloat(Minimum=10, Maximum=20))
   &SkeletalMeshLocation(
    Bone / Socket Index=&SelectIntFromArray(Direct Array Index=$Particles.PartitionParticles.Partition, Int Selection Array=DataInterface() {propertyValues {IntData [0,1,2,3,4,5,6]}})
    Bone Sampling Mode=::NewEnumerator4
    MaskSpawnGroup=false
    Orientation Sampling=::NewEnumerator1
    Sampled Position Offset Coordinate Space:ENiagaraSkelMeshTransforms=::NewEnumerator1
    Skeletal Mesh=$System.Skeletal Mesh
   ) {module=SkeletalMeshLocation003}
   &DynamicMaterialParameters(
    Index 0 Param 1=&SelectFloatFromArray(Direct Array Index=$Particles.PartitionParticles.Partition, Float Selection Array=DataInterface() {propertyValues {FloatData [5,5,4,4,3,3]}})
    Index 0 Param 3=1
    Module.Index0Float0Write=true
    Module.Index0Float1Write=true
    Module.Index0Float2Write=false
    Module.Index0Float3Write=true
   )
   /SetVariables(Particles.glowColor:LinearColor=&MakeLinearColorFromVectorAndFloat(
     "Vector (RGB)"=&MakeVectorFromLinearColorRGB(Color=<rgba 0.4031 2.0469 3 1>)
     "Float (Alpha)"=&FloatFromCurve(
      FloatCurve=DataInterface() {propertyValues {Curve {keys [<at 0 1>,<at 1 1>]}}}
      Scale Curve=&MakeCustomFloatFromBool(Bool=$User.IsFrontEnd, True Float=0.12, False Float=0.15)
     )
    )) {isSetParameters=true, module=SetVariables_#5}
  ]
  rendererClasses [&NiagaraSpriteRendererProperties]
  renderers [
   {rendererClass=&NiagaraSpriteRendererProperties}
   {rendererClass=&NiagaraSpriteRendererProperties, rendererIndex=1}
  ]
 }
 Graffiti_Icons {
  emitterUpdateScript [
   &EmitterState()
   &SpawnRate(Spawn Probability=0.85, SpawnRate=2, Use Spawn Probability=true)
   &SpawnBurst_Instantaneous(Spawn Count=2, Use Spawn Probability=false, UseLoopCountLimit=true)
  ]
  particleSpawnScript [
   &PartitionParticles(Number Of Partitions=4)
   &InitializeParticle(
    Color=&ColorFromCurve(
     ColorCurve=DataInterface() {propertyValues {
       AlphaCurve {keys [<at 0 1 RCIM_Cubic>]}
       BlueCurve {keys [<at 0 1 RCIM_Cubic>,<at 1 1 RCIM_Cubic>]}
       GreenCurve {keys [<at 0 0.294 RCIM_Cubic>,<at 1 0.9851 RCIM_Cubic>]}
       RedCurve {keys [<at 0 0.0802 RCIM_Cubic>,<at 1 0.0594 RCIM_Cubic>]}
      }}
     CurveIndex=&RandomRangeFloat(Minimum=0, Maximum=1)
    )
    Lifetime=&RandomRangeFloat(Minimum=1.7, Maximum=1.85)
    Sprite Rotation Angle=&RandomRangeFloat(Minimum=-90, Maximum=90)
    Sprite Rotation Mode=::NewEnumerator0
    Sprite Size Mode=::NewEnumerator3
    Sprite UV Mode=::NewEnumerator1
    Uniform Sprite Size=&FloatFromCurve(
     FloatCurve=DataInterface() {propertyValues {Curve {keys [<at 0 1>,<at 0.7 1.2>]}}}
     CurveIndex=&RandomRangeFloat(Minimum=0, Maximum=1)
     Scale Curve=&SelectFloatFromArray(
      Direct Array Index=$Particles.PartitionParticles.Partition
      Direct Array Mode=::NewEnumerator1
      Float Selection Array=DataInterface() {propertyValues {FloatData [13,17,14,19]}}
     )
    )
    UsePositionOffset=false
    Write Material Random=true
   )
   &CameraOffset(Camera Offset Amount=-55)
   &AddVelocity(Velocity=<0 0 1>, Velocity Speed Scale=&RandomRangeFloat(Minimum=25, Maximum=75))
  ]
  particleUpdateScript [
   &ParticleState()
   &SubUVAnimation(
    Define SubUV Setup Manually:ENiagara_AutomaticManual=::NewEnumerator1
    End Frame Range Override:NiagaraInt32=&SelectIntFromArray(
     Direct Array Index=$Particles.PartitionParticles.Partition
     Direct Array Mode=::NewEnumerator1
     Int Selection Array=DataInterface() {propertyValues {IntData [1,4,1]}}
    )
    Start Frame Range Override=&SelectIntFromArray(
     Direct Array Index=$Particles.PartitionParticles.Partition
     Direct Array Mode=::NewEnumerator1
     Int Selection Array=DataInterface() {propertyValues {IntData [0,3,0]}}
    )
    SubUV Animation Mode:ENiagara_SubUVLookupModeV2=::NewEnumerator1
    SubUV Curve=DataInterface() {propertyValues {Curve {
       defaultValue=0
       keys [
        <at 0.45 0 RCIM_Constant>
        <at 0.5 0.5 RCIM_Constant>
        <at 0.55 1 RCIM_Constant>
        <at 0.95 1 RCIM_Constant>
        {arriveTangent=9.0725, interpMode=RCIM_Constant, time=0.96, value=0.56}
       ]
      }}}
    UseEndFrame=true
    UseStartFrame=true
   )
   /SetVariables(Particles.SpriteRotation:NiagaraFloat=&Lerp_Float(
     A=$Particles.Initial.SpriteRotation
     B=&RandomRangeFloat(Minimum=-5, Maximum=5)
     Alpha=&FloatFromCurve(FloatCurve=DataInterface() {propertyValues {Curve {
         defaultValue=1
         keys [
          {arriveTangent=3.5971, interpMode=RCIM_Cubic, leaveTangent=3.5971, tangentMode=RCTM_User, time=0, value=0}
          {arriveTangentWeight=0.3797, interpMode=RCIM_Cubic, leaveTangentWeight=0.2531, tangentMode=RCTM_User, time=0.6, value=1}
          {arriveTangent=2.3632, interpMode=RCIM_Cubic, leaveTangent=2.3632, tangentMode=RCTM_User, time=1, value=1.6137}
         ]
        }}})
    )) {isSetParameters=true}
   &ScaleSpriteSize(
    Uniform Curve Scale=&RandomRangeFloat(Minimum=1, Maximum=1.2)
    Uniform Curve Sprite Scale=DataInterface() {propertyValues {Curve {keys [
        {arriveTangent=11.8857, interpMode=RCIM_Cubic, leaveTangent=11.8857, tangentMode=RCTM_User, time=0, value=0}
        {arriveTangent=3.7041, interpMode=RCIM_Cubic, leaveTangent=3.7041, tangentMode=RCTM_User, time=0.1, value=1}
        {arriveTangent=-6.5017, interpMode=RCIM_Cubic, leaveTangent=-6.5017, tangentMode=RCTM_User, time=0.45, value=1.0324}
        {arriveTangent=-0.1765, interpMode=RCIM_Cubic, leaveTangent=-0.1765, time=0.5, value=0.5752}
        {arriveTangent=2.5279, interpMode=RCIM_Cubic, leaveTangent=2.5279, tangentMode=RCTM_User, time=0.55, value=1.0147}
        {arriveTangent=0.7713, interpMode=RCIM_Cubic, leaveTangent=0.7713, time=0.95, value=1.1}
        {arriveTangent=3.1211, interpMode=RCIM_Cubic, leaveTangent=3.1211, tangentMode=RCTM_User, time=1.0033, value=1.3644}
       ]}}}
   )
   &SkeletalMeshLocation(
    Bone / Socket Index=7
    Bone Sampling Mode=::NewEnumerator4
    MaskSpawnGroup=false
    Orientation Sampling=::NewEnumerator1
    Position Sampling=::NewEnumerator1
    Sampled Position Offset=&SelectVectorFromArray(
     Direct Array Index=$Particles.PartitionParticles.Partition
     Vector Selection Array=DataInterface() {propertyValues {FloatData [<25 25 -20>, <22 -25 20>, <30 10 25>, <25 -10 -5>, <45 20 -25>, <35 -20 5>]}}
    )
    Skeletal Mesh=$System.Skeletal Mesh
   ) {enabled=false, module=SkeletalMeshLocation001}
   &SkeletalMeshLocation(
    Bone / Socket Index=&SelectIntFromArray(Direct Array Index=$Particles.PartitionParticles.Partition, Int Selection Array=DataInterface() {propertyValues {IntData [7,7,7,7]}})
    Bone Sampling Mode=::NewEnumerator4
    MaskSpawnGroup=false
    Position Sampling=::NewEnumerator1
    Sampled Position Offset=&Add_Vector(
     A=&RandomRangeVector(Minimum=<-2 0 -5>, Maximum=<6 5 5>)
     B=&Multiply_Vector(
      A=&SelectVectorFromArray(
       Direct Array Index=$Particles.PartitionParticles.Partition
       Vector Selection Array=DataInterface() {propertyValues {FloatData [<15 0 -15>, <15 0 15>, <3 0 -28>, <3 0 28>]}}
      )
      B=&MakeCustomVectorFromBool(Bool=$User.IsLargeBody, True Vector=<1.55 1 1.45>, False Vector=<1 1 1.2>)
     )
    )
    Skeletal Mesh=$System.Skeletal Mesh
   )
   &InheritSourceMovement(
    Applied Position Scale=&Multiply_Vector(A=<-0.25 -0.25 -0.5>, B=&VectorFromFloat(Value=0.6))
    ApplyForce=false
    ApplyPosition=true
    ApplyPositionFalloff=false
   )
   &Drag(Drag=&RandomRangeFloat(Minimum=1, Maximum=2), Use Linear Drag=true, Use Rotational Drag=true)
   &SolveForcesAndVelocity(Clamp Velocity=false, Limit Acceleration=false)
   &ScaleColor(
    Scale Alpha=&FloatFromCurve(FloatCurve=DataInterface() {propertyValues {Curve {keys [<at 0 1>,<at 0.45 1>,<at 0.5027 0.7>,<at 0.5446 1>,<at 0.95 1>,<at 1 0>]}}})
    Scale RGB=&VectorFromFloat(Value=&FloatFromCurve(
      FloatCurve=DataInterface() {propertyValues {Curve {keys [
          <at 0 1 RCIM_Cubic>
          {arriveTangent=7.8221, interpMode=RCIM_Cubic, leaveTangent=7.8221, tangentMode=RCTM_User, time=0.3923, value=2}
          {arriveTangent=-1.2343, interpMode=RCIM_Cubic, leaveTangent=-1.2343, time=0.4984, value=1}
          {arriveTangent=0.0695, interpMode=RCIM_Cubic, leaveTangent=0.0695, time=0.5949, value=1.75}
          {arriveTangent=-2.0513, interpMode=RCIM_Cubic, leaveTangent=-2.0513, time=0.7983, value=1.0209}
          <at 1 0.919 RCIM_Cubic>
         ]}}}
      Scale Curve=&FNTimeOfDayScalar(Night Scale=1.15, Morning Scale=0.9, Day Scale=0.5, Evening Scale=0.9)
     ))
    ScaleA=true
    ScaleRGB=true
   )
   /SetVariables(
    Particles.Color2=&MultiplyLinearColor(
     Linear Color=&MultiplyLinearColor(
      Linear Color=$Particles.Color
      Scale Factor=&MakeVector4_FromLinearColor(LinearColor=$Particles.Initial.Color)
     )
     Scale Factor=<0.6 0.7 0.7 0.55>
    )
    Particles.FlareColor=&MultiplyLinearColor(
     Linear Color=&MultiplyLinearColor(
      Linear Color=$Particles.Color
      Scale Factor=&MakeVector4_FromLinearColor(LinearColor=$Particles.Initial.Color)
     )
     Scale Factor=<1.1 1.1 1.1 0.7>
    )
    Particles.FlareSize:Vector2f=&Multiply_Vector2D_ByFloat(Float=2.5, Vector2D=$Particles.SpriteSize)
    Particles.GlowSize=&Multiply_Vector2D_ByFloat(Float=1.4, Vector2D=$Particles.SpriteSize)
   ) {isSetParameters=true}
   /SetVariables(Particles.position modified:NiagaraPosition=&AddVectorToPosition(
     Position=$Particles.Position
     Vector=&Multiply_VectorByFloat(
      Vector=&NormalizeVector(
       Vector To Normalize=&SubtractVectorFromPosition(
        Position=$Particles.Position
        Vector=Linked() {linkedVariable {name=Output.SkeletalMeshLocation.SampledPositionWithoutOffset, type=NiagaraPosition}}
       )
       Safe Normalize=true
       Safe Fallback Threshold=0
       Safe Fallback Vector=<0 0 0.001>
      )
      Float=&FloatFromCurve(
       FloatCurve=DataInterface() {propertyValues {Curve {
          defaultValue=0
          keys [
           {arriveTangent=2.7405, interpMode=RCIM_Cubic, leaveTangent=2.7405, tangentMode=RCTM_User, time=0, value=0}
           {arriveTangent=0.4924, interpMode=RCIM_Cubic, leaveTangent=0.4924, tangentMode=RCTM_User, time=0.65, value=0.9}
           {arriveTangentWeight=0.2204, interpMode=RCIM_Cubic, tangentMode=RCTM_User, time=0.9966, value=1}
          ]
         }}}
       Scale Curve=&Multiply_Float(
        A=&Multiply_Float(
         A=&FloatFromCurve(
          FloatCurve=DataInterface() {propertyValues {Curve {keys [<at 0 3>,<at 1 5>]}}}
          CurveIndex=&RandomRangeFloat(Minimum=0, Maximum=1)
         )
         B=&SelectFloatFromArray(
          Direct Array Index=$Particles.PartitionParticles.Partition
          Float Selection Array=DataInterface() {propertyValues {FloatData [0.6,0.6,0.4,0.4]}}
         )
        )
        B=&MakeCustomFloatFromBool(Bool=$User.IsLargeBody, True Float=1.1, False Float=1)
       )
      )
     )
    )) {isSetParameters=true}
   &InterpolateOverTime(Rate Of Change=5, Target Position Value:NiagaraPosition=$Particles.position modified, Variable Type=::NewEnumerator6)
   /SetVariables(Particles.Position=$Particles.InterpolateOverTime.InterpolateOverTime_Position.Moving Average) {isSetParameters=true}
  ]
  rendererClasses [&NiagaraSpriteRendererProperties]
  renderers [
   {rendererClass=&NiagaraSpriteRendererProperties}
   {rendererClass=&NiagaraSpriteRendererProperties, rendererIndex=1}
   {rendererClass=&NiagaraSpriteRendererProperties, rendererIndex=2}
  ]
 }
 Rays {
  bEnabled=false
  emitterUpdateScript [
   &EmitterState()
   &SpawnBurst_Instantaneous(Use Spawn Probability=false, UseLoopCountLimit=true)
  ]
  particleSpawnScript [
   &InitializeParticle(Color=<rgba 1.25 1.25 1.25 1>, Mesh Scale Mode=::NewEnumerator3, Mesh Uniform Scale=1.5, UsePositionOffset=false, Write Material Random=true)
   &DynamicMaterialParameters(
    Index 0 Param 1=&RandomRangeFloat(Minimum=0, Maximum=10)
    Index 0 Param 2=&RandomRangeFloat(Minimum=1, Maximum=2)
    Index 0 Param 3=&RandomRangeFloat(Minimum=0.8, Maximum=1.6)
    Module.Index0Param0Write=true
    Module.Index0Param1Write=true
    Module.Index0Param2Write=true
    Module.Index0Param3Write=false
   )
   &SkeletalMeshLocation(
    Bone / Socket Index=7
    Bone Sampling Mode=::NewEnumerator4
    MaskSpawnGroup=false
    Orientation Sampling=::NewEnumerator1
    Skeletal Mesh=$System.Skeletal Mesh
   )
   &CameraOffset(Camera Offset Amount=-60)
   &InitialMeshOrientation(Mesh Orientation Mode=::NewEnumerator2, Rotation=<0.5 0 0>, Use Rotation Vector=true)
  ]
  particleUpdateScript [
   &ParticleState(Kill Particles When Lifetime Has Elapsed=false)
   &SkeletalMeshLocation(
    Bone / Socket Index=7
    Bone Sampling Mode=::NewEnumerator4
    MaskSpawnGroup=false
    Orientation Sampling=::NewEnumerator1
    Skeletal Mesh=$System.Skeletal Mesh
   ) {module=SkeletalMeshLocation001}
   &ScaleColor(
    Scale Alpha=&MakeCustomFloatFromBool(Bool=$User.IsFrontEnd, True Float=0.21, False Float=0.6)
    Scale RGB=&Multiply_VectorByFloat(
     Vector=&MakeVectorFromLinearColorRGB(Color=&ColorFromCurve(
       ColorCurve=DataInterface() {propertyValues {
         AlphaCurve {keys []}
         BlueCurve {keys [<at 0 1 RCIM_Cubic>,<at 0.2994 1.2>,<at 0.5042 1>,<at 0.7224 1.2>,<at 1 1 RCIM_Cubic>]}
         GreenCurve {keys [<at 0 0.4342 RCIM_Cubic>,<at 0.2994 0.4103>,<at 0.5042 0.8482>,<at 0.7224 0.4103>,<at 1 0.4342 RCIM_Cubic>]}
         RedCurve {keys [<at 0 0.9387 RCIM_Cubic>,<at 0.2994 0>,<at 0.5042 0.1742>,<at 0.7224 0>,<at 1 0.9387 RCIM_Cubic>]}
        }}
       CurveIndex=&Frac_Float(Float=&Multiply_Float(A=$Engine.Time, B=0.5))
      ))
     Float=&Multiply_Float(A=&FNTimeOfDayScalar(Night Scale=1.1, Morning Scale=1.1, Day Scale=1.1, Evening Scale=1.1), B=1)
    )
    ScaleA=true
    ScaleRGB=true
   ) {module=ScaleColor001}
  ]
  rendererClasses [&NiagaraMeshRendererProperties]
  renderers [{rendererClass=&NiagaraMeshRendererProperties, rendererIndex=0}]
 }
 SpeedLine {
  emitterUpdateScript [
   &EmitterState()
   &SpawnRate(Spawn Probability=0.95, SpawnRate=2.75, Use Spawn Probability=true)
   &SpawnBurst_Instantaneous(Spawn Count=1, Use Spawn Probability=false)
  ]
  particleSpawnScript [
   &InitializeParticle(
    Color=<rgba 15 15 15 1>
    Lifetime=&RandomRangeFloat(Minimum=0.85, Maximum=0.95)
    Sprite Rotation Mode=::NewEnumerator2
    Sprite Size Mode=::NewEnumerator3
    Uniform Sprite Size=&Multiply_Float(
     A=&RandomRangeFloat(Minimum=125, Maximum=130)
     B=&MakeCustomFloatFromBool(Bool=$User.IsLargeBody, True Float=1.15, False Float=1)
    )
    UsePositionOffset=false
    Write Material Random=true
   ) {module=InitializeParticle001}
   &CameraOffset(Camera Offset Amount=-50)
  ]
  particleUpdateScript [
   &ParticleState()
   &SkeletalMeshLocation(
    Bone / Socket Index=7
    Bone Sampling Mode=::NewEnumerator4
    MaskSpawnGroup=false
    Orientation Sampling=::NewEnumerator1
    Skeletal Mesh=$System.Skeletal Mesh
   )
   &SolveForcesAndVelocity(Clamp Velocity=false, Limit Acceleration=false)
   &ScaleSpriteSize(Scale Factor=&Vector2DFromFloat(Value=&FloatFromCurve(FloatCurve=DataInterface() {propertyValues {Curve {keys [{arriveTangent=2.971, interpMode=RCIM_Cubic, leaveTangent=2.971, tangentMode=RCTM_User, time=0, value=0}, <at 1 1 RCIM_Cubic>]}}})))
   &DynamicMaterialParameters(
    Index 0 Param 1=&FloatFromCurve(FloatCurve=DataInterface() {propertyValues {Curve {
        defaultValue=0.5
        keys [
         {arriveTangent=-0.0128, interpMode=RCIM_Cubic, leaveTangent=-0.0128, tangentMode=RCTM_User, time=0, value=1}
         {arriveTangent=-0.1899, interpMode=RCIM_Cubic, leaveTangent=-0.1899, tangentMode=RCTM_User, time=1, value=0.6}
        ]
       }}})
    Module.Param0WriteEnabled=true
    Module.Param1WriteEnabled=true
    Module.Param2WriteEnabled=true
    Module.Param3WriteEnabled=true
   )
   &ScaleColor(
    Scale Alpha=&FloatFromCurve(
     FloatCurve=DataInterface() {propertyValues {Curve {keys [
         {
          arriveTangent=-12.3944
          interpMode=RCIM_Cubic
          leaveTangent=-12.3944
          leaveTangentWeight=0.8844
          tangentMode=RCTM_User
          time=0.003
          value=4.9576
         }
         {arriveTangentWeight=1.2996, interpMode=RCIM_Cubic, tangentMode=RCTM_User, time=1, value=0}
        ]}}}
     Scale Curve=3
    )
    ScaleA=true
    ScaleRGB=true
   )
   &ScaleColor(
    Scale Alpha=&MakeCustomFloatFromBool(Bool=$User.IsFrontEnd, True Float=0.7, False Float=1)
    Scale RGB=&Multiply_VectorByFloat(
     Vector=&MakeVectorFromLinearColorRGB(Color=&ColorFromCurve(
       ColorCurve=DataInterface() {propertyValues {
         AlphaCurve {keys []}
         BlueCurve {keys [<at 0 1 RCIM_Cubic>,<at 0.2994 1>,<at 0.5042 1>,<at 0.7224 1>,<at 1 1 RCIM_Cubic>]}
         GreenCurve {keys [<at 0 0.4342 RCIM_Cubic>,<at 0.2994 0.3419>,<at 0.5042 0.8482>,<at 0.7224 0.3419>,<at 1 0.4342 RCIM_Cubic>]}
         RedCurve {keys [<at 0 0.9387 RCIM_Cubic>,<at 0.2994 0>,<at 0.5042 0.1742>,<at 0.7224 0>,<at 1 0.9387 RCIM_Cubic>]}
        }}
       CurveIndex=&Frac_Float(Float=&Multiply_Float(A=$Engine.Time, B=0.5))
      ))
     Float=&FNTimeOfDayScalar(Night Scale=1.2, Morning Scale=1, Day Scale=1.1, Evening Scale=1.1)
    )
    ScaleA=true
    ScaleRGB=true
   ) {module=ScaleColor001}
  ]
  rendererClasses [&NiagaraSpriteRendererProperties]
  renderers [{rendererClass=&NiagaraSpriteRendererProperties}]
 }
 Tiny_StarsandDots {
  emitterUpdateScript [
   &EmitterState()
   &SpawnRate(SpawnGroup=1, SpawnRate=10, Use Spawn Probability=false) {module=SpawnRate001}
   &SpawnRate(SpawnRate=10, Use Spawn Probability=false)
  ]
  particleSpawnScript [
   &PartitionParticles(Number Of Partitions=3)
   &SkeletalMeshLocation(
    Bone / Socket Index=&SelectIntFromArray(Direct Array Index=$Particles.PartitionParticles.Partition, Int Selection Array=DataInterface() {propertyValues {IntData [1,2,3,4]}})
    Bone Sampling Mode=::NewEnumerator4
    MaskSpawnGroup=false
    Orientation Sampling=::NewEnumerator1
    Sampled Position Offset Coordinate Space:ENiagaraSkelMeshTransforms=::NewEnumerator2
    Skeletal Mesh=$System.Skeletal Mesh
   ) {enabled=false, module=SkeletalMeshLocation003}
   &SkeletalMeshLocation(
    Bone / Socket Index=7
    Bone Sampling Mode=::NewEnumerator4
    MaskSpawnGroup=false
    Position Sampling=::NewEnumerator2
    Sampled Position Offset Coordinate Space:ENiagaraSkelMeshTransforms=::NewEnumerator2
    Skeletal Mesh=$System.Skeletal Mesh
   ) {module=SkeletalMeshLocation001}
   &SkeletalMeshLocation(
    Bone / Socket Index=&SelectIntFromArray(Direct Array Index=$Particles.PartitionParticles.Partition, Int Selection Array=$System.companion bone)
    Bone Sampling Mode=::NewEnumerator4
    MaskSpawnGroup=false
    Orientation Sampling=::NewEnumerator1
    Sampled Position Offset Coordinate Space:ENiagaraSkelMeshTransforms=::NewEnumerator2
    Skeletal Mesh=$System.Skeletal Mesh
   ) {enabled=false}
   /SetVariables(
    Particles.bonesOffset=DataInterface() {propertyValues {GpuSyncMode=SyncCpuToGpu, MaxElements=0, FloatData [<-20 0 20>, <25 -5 1>, <-25 5 -1>]}}
    Particles.positionlerp:NiagaraFloat=&RandomRangeFloat(Minimum=0, Maximum=1)
    Particles.postionOffset:Vector3f=&Multiply_VectorByFloat(
     Vector=&MakeCustomVectorFromBool(
      Bool=$User.IsFrontEnd
      True Vector=&RandomRangeVector(Minimum=<0 0 1>, Maximum=<0 0 2>)
      False Vector=<0 -1 1>
     )
     Float=&MakeCustomFloatFromBool(
      Bool=$User.IsLargeBody
      True Float=&RandomRangeFloat(Minimum=9, Maximum=15)
      False Float=&RandomRangeFloat(Minimum=2, Maximum=4)
     )
    )
   ) {isSetParameters=true}
   &InitializeParticle(
    Color=&ColorFromCurve(
     ColorCurve=DataInterface() {propertyValues {
       AlphaCurve {keys [<at 0 1 RCIM_Cubic>]}
       BlueCurve {keys [<at 0 1 RCIM_Cubic>,<at 1 1 RCIM_Cubic>]}
       GreenCurve {keys [<at 0 0.294 RCIM_Cubic>,<at 1 0.9851 RCIM_Cubic>]}
       RedCurve {keys [<at 0 0.0802 RCIM_Cubic>,<at 1 0.0594 RCIM_Cubic>]}
      }}
     CurveIndex=&RandomRangeFloat(Minimum=0, Maximum=1)
    )
    Lifetime Max=0.9
    Lifetime Min=0.5
    Lifetime Mode=::NewEnumerator1
    Sprite Rotation Mode=::NewEnumerator2
    Sprite Size Mode=::NewEnumerator3
    Uniform Sprite Size=&MaskFloatBySpawnGroup(
     Masked Float=&Multiply_Float(
      A=&RandomRangeFloat(Minimum=2, Maximum=6)
      B=&MakeCustomFloatFromBool(Bool=$User.IsLargeBody, True Float=1.1, False Float=1)
     )
     Passthrough Float=&Multiply_Float(
      A=&RandomRangeFloat(Minimum=1.5, Maximum=4)
      B=&MakeCustomFloatFromBool(Bool=$User.IsLargeBody, True Float=1.1, False Float=1)
     )
    )
    UsePositionOffset=false
    Write Material Random=true
   )
   &SkeletalMeshLocation(
    Bone / Socket Index=&SelectIntFromArray(Direct Array Index=$Particles.PartitionParticles.Partition, Int Selection Array=DataInterface() {propertyValues {IntData [0,1,2]}})
    Bone Sampling Mode=::NewEnumerator4
    MaskSpawnGroup=false
    Orientation Sampling=::NewEnumerator1
    Position Sampling=::NewEnumerator1
    Sampled Position Offset=&LerpPosition(
     A=<0 0 0>
     B=&SelectVectorFromArray(Direct Array Index=$Particles.PartitionParticles.Partition, Vector Selection Array=$Particles.bonesOffset)
     Alpha=$Particles.positionlerp
    )
    Skeletal Mesh=$System.Skeletal Mesh
   ) {module=SkeletalMeshLocation004}
   /SetVariables() {isSetParameters=true}
   &CameraOffset(Camera Offset Amount=&MakeCustomFloatFromBool(
     Bool=&SetBoolByIntComparison(A=$Particles.PartitionParticles.Partition, B=1, Comparison Type=::NewEnumerator7)
     True Float=&RandomRangeFloat(Minimum=-30, Maximum=-35)
     False Float=&SelectFloatFromArray(
      Array Sampling Mode=::NewEnumerator0
      Direct Array Index=?
      Float Selection Array=DataInterface() {propertyValues {FloatData [-30,-25,-20,-15,-17]}}
      Random Seed=0
     )
    ))
   &CameraOffset(Camera Offset Amount=40) {enabled=false, module=CameraOffset001}
   /SetVariables(Particles.cameraoffset2=-20) {enabled=false, isSetParameters=true}
   &AddVelocity(
    Velocity=&Multiply_Vector(
     A=&SelectVectorFromArray(
      Array Sampling Mode=::NewEnumerator0
      Direct Array Index=?
      Vector Selection Array=DataInterface() {propertyValues {FloatData [<0 0 10>, <0 0 1>, <0 0 -90>]}}
      Random Seed=0
     )
     B=&VectorFromFloat(Value=1.5)
    )
    Velocity Speed Scale=&MakeCustomFloatFromBool(Bool=$User.IsLargeBody, True Float=2, False Float=1)
   ) {module=AddVelocity001}
   &AddVelocity(
    Module.Use Falloff Distance=false
    Override Seed=false
    Velocity Mode=::NewEnumerator1
    Velocity Origin=$Output.SkeletalMeshLocation001.SampledPositionWithoutOffset
    Velocity Speed=&Multiply_Float(
     A=&Multiply_Float(
      A=&MakeCustomFloatFromBool(
       Bool=$User.IsFrontEnd
       True Float=&RandomRangeFloat(Minimum=70, Maximum=100)
       False Float=&RandomRangeFloat(Minimum=50, Maximum=90)
      )
      B=&MakeCustomFloatFromBool(Bool=$User.IsLargeBody, True Float=1.2, False Float=1)
     )
     B=1.75
    )
   )
   &SubUVAnimation(
    Direct Frame Index:NiagaraFloat=&MaskFloatBySpawnGroup(
     Masked Float=&SelectFloatFromArray(
      Array Sampling Mode=::NewEnumerator0
      Direct Array Index=?
      Float Selection Array=DataInterface() {propertyValues {FloatData [0,6,7]}}
      Random Seed=0
     )
     Passthrough Float=8
    )
    "Sprite Renderer "=DataInterface() {propertyValues {SpriteRenderer=&self:Tiny_StarsandDots_0.NiagaraSpriteRendererProperties_1}}
    SubUV Animation Mode:ENiagara_SubUVLookupModeV2=::NewEnumerator5
   )
  ]
  particleUpdateScript [
   &ParticleState()
   &SpriteRotationRate(Rotation Rate=&RandomRangeFloat(Minimum=-100, Maximum=100))
   &SkeletalMeshLocation(
    Bone / Socket Index=&SelectIntFromArray(Direct Array Index=$Particles.PartitionParticles.Partition, Int Selection Array=DataInterface() {propertyValues {IntData [0,1,2]}})
    Bone Sampling Mode=::NewEnumerator4
    "Inherit Velocity (Scale)"=0.25
    Inherited Velocity Speed Limit=100
    Initialize Particle Position To Sampled Position=false
    MaskSpawnGroup=false
    Orientation Sampling=::NewEnumerator1
    Position Sampling=::NewEnumerator1
    Sampled Position Offset=&LerpPosition(
     A=&SelectVectorFromArray(
      Direct Array Index=$Particles.PartitionParticles.Partition
      Vector Selection Array=DataInterface() {propertyValues {FloatData [<-15 0 -15>, <0 0 0>, <0 0 0>]}}
     )
     B=&SelectVectorFromArray(Direct Array Index=$Particles.PartitionParticles.Partition, Vector Selection Array=$Particles.bonesOffset)
     Alpha=$Particles.positionlerp
    )
    Skeletal Mesh=$System.Skeletal Mesh
    Velocity Sampling=::NewEnumerator0
    VelocitySpeedLimit=true
   ) {module=SkeletalMeshLocation002}
   &InheritSourceMovement(
    Applied Position Falloff Distance=150
    Applied Position Scale=&Multiply_VectorByFloat(Vector=<1 1 0.25>, Float=-1)
    ApplyForce=false
    ApplyPosition=true
    ApplyPositionFalloff=true
    ApplyPositionFalloffCurve=false
    Source Velocity:Vector3f=$Particles.SkeletalMeshLocation002.SampledVelocity
   )
   &Drag(Drag=4, Use Linear Drag=true, Use Rotational Drag=true)
   &SolveForcesAndVelocity(Clamp Velocity=false, Limit Acceleration=false)
   &ScaleSpriteSize(Uniform Curve Sprite Scale=DataInterface() {propertyValues {Curve {
       defaultValue=0.85
       keys [
        {arriveTangent=6.7723, interpMode=RCIM_Cubic, leaveTangent=6.7723, tangentMode=RCTM_User, time=0, value=0.1}
        {arriveTangent=-0.0348, interpMode=RCIM_Cubic, leaveTangent=-0.0348, tangentMode=RCTM_User, time=0.1992, value=1.0113}
        {arriveTangent=-0.6685, interpMode=RCIM_Cubic, leaveTangent=-0.6685, time=0.7001, value=1.1539}
        {
         arriveTangent=-2.3254
         arriveTangentWeight=0.2667
         interpMode=RCIM_Cubic
         leaveTangent=-2.3254
         tangentMode=RCTM_User
         time=0.9837
         value=0.4868
        }
       ]
      }}})
   &InterpolateOverTime(Rate Of Change=5, Target Position Value:NiagaraPosition=$Output.SolveForcesAndVelocity.Position, Variable Type=::NewEnumerator6)
   &ScaleColor(
    Linear Color Curve=DataInterface() {propertyValues {
      AlphaCurve {keys [<at 0 0>,<at 0.123 1>,<at 0.7333 1>,<at 1 0>]}
      BlueCurve {keys [<at 0.1185 3>,<at 0.9946 0.5>]}
      GreenCurve {keys [<at 0.1185 3>,<at 0.9946 0.5>]}
      RedCurve {keys [<at 0.1185 3>,<at 0.9946 0.5>]}
     }}
    Scale Mode=::NewEnumerator2
   )
   &DynamicMaterialParameters(
    Index 0 Param 1=&FloatFromCurve(
     FloatCurve=DataInterface() {propertyValues {Curve {defaultValue=1, keys [<at 0.5 0.05>, {tangentMode=RCTM_User, time=0.7, value=0.05}, {tangentMode=RCTM_User, time=1, value=1}]}}}
     Scale Curve=&SelectFloatFromArray(
      Array Sampling Mode=::NewEnumerator0
      Direct Array Index=?
      Float Selection Array=DataInterface() {propertyValues {FloatData [0,2,2]}}
      Random Seed=0
     )
    )
    Index 0 Param 2=&FloatFromCurve(FloatCurve=DataInterface() {propertyValues {Curve {
        defaultValue=1
        keys [
         {arriveTangent=8.2781, interpMode=RCIM_Cubic, leaveTangent=8.2781, tangentMode=RCTM_User, time=0.2, value=0}
         {arriveTangent=1.1132, interpMode=RCIM_Cubic, leaveTangent=1.1132, tangentMode=RCTM_User, time=0.8, value=3}
        ]
       }}})
    Module.Index0Float0Write=true
    Module.Index0Float1Write=true
    Module.Index0Float2Write=true
    Module.Index0Float3Write=true
   )
   /SetVariables(Particles.Position=&AddVectorToPosition(
     Position=$StackContext.InterpolateOverTime.InterpolateOverTime_Position.Moving Average
     Vector=&Multiply_VectorByFloat(
      Vector=&MakeCustomVectorFromBool(
       Bool=$User.IsFrontEnd
       True Vector=&RandomRangeVector(Minimum=<0 0 1>, Maximum=<0 0 1.5>)
       False Vector=&RandomRangeVector(Minimum=<0 -5 1>, Maximum=<0 5 1>)
      )
      Float=&MakeCustomFloatFromBool(
       Bool=$User.IsLargeBody
       True Float=&RandomRangeFloat(Minimum=1.25, Maximum=1.5)
       False Float=&RandomRangeFloat(Minimum=0.5, Maximum=1.25)
      )
     )
    )) {isSetParameters=true}
  ]
  rendererClasses [&NiagaraSpriteRendererProperties]
  renderers [{rendererClass=&NiagaraSpriteRendererProperties}]
 }
}
summary {
 systemName=NS_Aura_DigitalGraffiti
 emitters [
  {bEnabled=false, emitterName=Glow001, rendererClasses [&NiagaraSpriteRendererProperties]}
  {bEnabled=false, emitterName=Rays, rendererClasses [&NiagaraMeshRendererProperties]}
  {emitterName=Arrows001, rendererClasses [&NiagaraMeshRendererProperties]}
  {emitterName=SpeedLine, rendererClasses [&NiagaraSpriteRendererProperties]}
  {emitterName=Tiny_StarsandDots, rendererClasses [&NiagaraSpriteRendererProperties]}
  {emitterName=Graffiti_Icons, rendererClasses [&NiagaraSpriteRendererProperties]}
  {emitterName=Glow, rendererClasses [&NiagaraSpriteRendererProperties, &NiagaraMeshRendererProperties]}
 ]
 userVariables [
  User.Interpolate Intensity {type=NiagaraFloat, defaultValue=5, description=""}
  User.IsFrontEnd {type=NiagaraBool}
  User.IsLargeBody {type=NiagaraBool}
  User.SkeletalMesh {
   type=NiagaraDataInterfaceSkeletalMesh
   defaultValue {
    struct=NiagaraExt_VariableValue_DataInterface
    value {dataInterface=&self:NiagaraDataInterfaceSkeletalMesh_0}
   }
  }
 ]
}
